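# Translucent, two-sided effect material for the ChromeB vehicle wheel, drawing animated waves in a flat AccentColor emissive.
# Opposing panned samples of the wave mask combine with rerouted inputs and are hard-thresholded into wave opacity, scaled by WaveOpacity.
# Vertex red switches opacity to a vertical UV gradient with scrolling caustics, faded by vertex alpha.
# Time-sampled Perlin noise feeds named flicker reroutes; AccentColorStrength is left unconnected.
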Material /VehicleCosmetics/Wheels/ChromeB/Materials/MAT_Wheel_ChromeB_Waves_fx
graph {
 properties {
  bAutomaticallySetUsageInEditor=false
  bUseMaterialAttributes=true
  bUsedWithMeshDeformer=true
  bUsedWithSkeletalMesh=true
  blendMode=BLEND_Translucent
  twoSided=true
  usedShadingModels=Default Lit
 }
 nodes [
  n1 = VectorParameter() {parameterName=AccentColor, defaultValue=<rgba 0 0.3606 1 1>}
  n2 = Constant()
  n3 = If(A=VertexColor().R, B=Constant() {r=0.51}, "A > B"=Constant() {r=1}, "A == B"=@n2, "A < B"=@n2)
  n4 = Multiply(A=Time(), B=ScalarParameter() {defaultValue=0.5, parameterName=pannerspeed})
  n5 = TextureSample(UVs=Panner(Coordinate=TextureCoordinate() {uTiling=0.8}, Time=@n4) {speedX=-0.5}) {texture=/VehicleCosmetics/Wheels/ChromeB/Textures/T_ChromeB_Waves_M}
  n6 = NamedRerouteUsage()
  n7 = NamedRerouteUsage()
  n8 = OneMinus(None=@n5.A)
  n9 = Panner(Coordinate=TextureCoordinate() {uTiling=0.7}, Time=@n4) {speedX=0.5}
  n10 = TextureSample(UVs=Panner(Coordinate=TextureCoordinate(), Time=@n4) {speedX=-0.3}) {texture=&T_ChromeB_Waves_M}
  n11 = TextureSample(UVs=Panner(Coordinate=TextureCoordinate() {uTiling=1.3}, Time=@n4) {speedX=0.5}) {texture=&T_ChromeB_Waves_M}
  n12 = Add(
   A=@n5.B
   B=Add(A=TextureSample(UVs=@n9) {texture=&T_ChromeB_Waves_M}.G, B=Multiply(A=@n10.G, B=@n11.R))
  )
  n13 = Add(A=Multiply(A=Add(A=@n6, B=@n7), B=Clamp(None=@n8)), B=Divide(A=Add(A=@n12, B=@n11.B)) {constB=3})
  n14 = Constant() {r=0.7}
  n15 = Constant() {r=1}
  n16 = Constant()
  n17 = Multiply(
   A=Power(Base=@n5.A, Exp=Constant() {r=0.7})
   B=Multiply(A=If(A=@n13, B=@n14, "A > B"=@n15, "A == B"=@n16, "A < B"=@n16), B=ScalarParameter() {defaultValue=0.85, parameterName=WaveOpacity})
  )
  n18 = ComponentMask(None=Power(Base=TextureCoordinate(), Exp=Constant() {r=4})) {g=true}
  n19 = TextureCoordinate() {uTiling=3, vTiling=0.2}
  n20 = Add(A=@n18, B=TextureSample(UVs=Panner(Coordinate=@n19) {speedY=0.1}) {texture=/VehicleCosmetics/Wheels/ChromeB/Textures/T_Radiant_Caustics_D}.R)
  n21 = Clamp(None=Power(Base=Multiply(A=@n18, B=@n20), Exp=Constant() {r=0.7}))
  n22 = VertexColor()
  n23 = Clamp(None=Power(Base=Add(A=@n22.A, B=@n22.A)))
  MakeMaterialAttributes!(
   BaseColor
   Metallic
   Specular
   Roughness
   Anisotropy
   EmissiveColor=LinearInterpolate(A=@n1.RGB, B=@n1.RGB, Alpha=@n3)
   Opacity=LinearInterpolate(A=Clamp(None=@n17), B=Multiply(A=@n21, B=@n23), Alpha=@n3) {constA=0.5}
   OpacityMask
   Normal
   Tangent
   WorldPositionOffset
   SubsurfaceColor
   ClearCoat
   ClearCoatRoughness
   AmbientOcclusion
   Refraction
   CustomizedUVs_0
   CustomizedUVs_1
   CustomizedUVs_2
   CustomizedUVs_3
   CustomizedUVs_4
   CustomizedUVs_5
   CustomizedUVs_6
   CustomizedUVs_7
   PixelDepthOffset
   ShadingModel
   Displacement
  )
  Multiply(A=@n1.RGB, B=ScalarParameter() {defaultValue=1, parameterName=AccentColorStrength})
  n24 = Multiply(A=Time(), B=ScalarParameter() {defaultValue=1, parameterName=TimeSpeedLOW Freq})
  n25 = Clamp(None=Multiply(
    A=TextureSample(UVs=@n24) {samplerType=SAMPLERTYPE_LinearColor, texture=/Game/Global/Textures/Default/Tileable_Patterns/T_PerlinNoise_Uncompressed}.R
    B=ScalarParameter() {defaultValue=1, parameterName=RandomLOWFreq}
   ))
  NamedRerouteDeclaration(None=Clamp(None=Clamp(None=@n25) {minDefault=0.3}) {maxDefault=1.5}) {name=LOW Freq out Clamp, nodeColor=<rgba 1 0.5156 0 1>}
  NamedRerouteDeclaration(None=Multiply(B=@n25) {constA=0.2}) {name=LOW Freq out Mult, nodeColor=<rgba 1 0.5156 0 1>}
  n26 = Multiply(A=Time(), B=ScalarParameter() {defaultValue=0.1, parameterName=TimeSpeedHighFreq})
  n27 = TextureSample(UVs=@n26) {samplerType=SAMPLERTYPE_LinearColor, texture=/Game/Weapons/FORT_Melee/Pickaxe_RhombCamo/FX/T_Pickaxe_RhombCamo_Perlin}
  n28 = ScalarParameter() {defaultValue=10, parameterName=RandomHighFreq}
  NamedRerouteDeclaration(None=Clamp(None=Multiply(A=@n27.R, B=@n28))) {name=HIGH Freq out Clamp, nodeColor=<rgba 1 0.5156 0 1>}
  Sine(None=@n26)
  Sine(None=@n24)
  VertexColor()
 ]
}
