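# Translucent material for the animated radiant glow on the ChromeB vehicle wheel.
# Two caustics samples pan over tileable UVs, are masked by a gradient texture, then added to a faint gradient with a scrolling pulse.
# Vertex alpha scales the result, which drives both opacity and emissive strength.
# The Painted static switch picks PaintColor or a gradient blend of CustomColor and AccentColor as the emissive tint.
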Material /VehicleCosmetics/Wheels/ChromeB/Materials/MAT_Wheel_Radiant_ChromeB
graph {
 properties {
  bAutomaticallySetUsageInEditor=false
  bUseMaterialAttributes=true
  bUsedWithMeshDeformer=true
  bUsedWithSkeletalMesh=true
  blendMode=BLEND_Translucent
  usedShadingModels=Default Lit
 }
 nodes [
  n1 = TextureSample() {texture=/VehicleCosmetics/Wheels/ChromeB/Textures/T_Radiant_Gradient}
  n2 = Add(A=@n1.R, B=@n1.R)
  n3 = StaticSwitchParameter(
   True=VectorParameter() {parameterName=PaintColor, defaultValue=<rgba 1 0.4223 0 1>}.RGB
   False=LinearInterpolate(
    A=VectorParameter() {parameterName=CustomColor, defaultValue=<rgba 1 0.1878 0 1>}.RGB
    B=VectorParameter() {parameterName=AccentColor, defaultValue=<rgba 0 0.2621 1 1>}.RGB
    Alpha=Clamp(None=@n2)
   )
  ) {parameterName=Painted}
  n4 = ScalarParameter() {defaultValue=25, parameterName=EmissivePower}
  n5 = TextureCoordinate()
  n6 = AppendVector(
   A=Multiply(A=ComponentMask(None=@n5) {r=true}, B=ScalarParameter() {defaultValue=1, parameterName=GlowTileX})
   B=Multiply(A=ComponentMask(None=@n5) {g=true}, B=ScalarParameter() {defaultValue=1, parameterName=GlowTileY})
  )
  n7 = Multiply(A=ScalarParameter() {defaultValue=1, parameterName=Glow_Panspeed}, B=Time())
  n8 = Panner(Coordinate=@n6, Time=@n7) {speedX=0.075, speedY=0.067}
  n9 = TextureSample(UVs=Panner(Coordinate=@n6, Time=@n7) {speedX=0.125, speedY=1}) {texture=/VehicleCosmetics/Wheels/ChromeB/Textures/T_Radiant_Caustics_D}
  n10 = Add(A=Multiply(A=TextureSample(UVs=@n8) {texture=&T_Radiant_Caustics_D}.R, B=@n9.R), B=@n9.R)
  n11 = Multiply(A=@n1.R, B=ScalarParameter() {defaultValue=0.1, parameterName=GradientOpacity})
  n12 = Multiply(A=@n11, B=TextureSample(UVs=Panner() {speedY=0.125}) {texture=&T_Radiant_Gradient}.R)
  n13 = Multiply(A=VertexColor().A, B=Add(A=Multiply(A=@n10, B=@n1.R), B=Add(A=@n11, B=@n12)))
  MakeMaterialAttributes!(
   BaseColor
   Metallic
   Specular
   Roughness
   Anisotropy
   EmissiveColor=Multiply(A=Multiply(A=@n3, B=@n4), B=@n13)
   Opacity=@n13
   OpacityMask
   Normal
   Tangent
   WorldPositionOffset
   SubsurfaceColor
   ClearCoat
   ClearCoatRoughness
   AmbientOcclusion
   Refraction
   CustomizedUVs_0
   CustomizedUVs_1
   CustomizedUVs_2
   CustomizedUVs_3
   CustomizedUVs_4
   CustomizedUVs_5
   CustomizedUVs_6
   CustomizedUVs_7
   PixelDepthOffset
   ShadingModel
   Displacement
  )
  Multiply()
  TextureSample()
 ]
}
