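# Translucent Default Lit material for the lens flare on the ChromeB vehicle wheel cosmetic, flagged for skeletal mesh and mesh deformer use.
# A single lens flare texture supplies both color and mask, with its alpha scaled down to give a faint Opacity.
# That same scaled alpha multiplies the texture RGB and the EmissiveStrength parameter to form the glow, output through material attributes.
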Material /VehicleCosmetics/Wheels/ChromeB/Materials/MAT_wheel_ChromeB_flare
graph {
 properties {
  bAutomaticallySetUsageInEditor=false
  bUseMaterialAttributes=true
  bUsedWithMeshDeformer=true
  bUsedWithSkeletalMesh=true
  blendMode=BLEND_Translucent
  usedShadingModels=Default Lit
 }
 nodes [
  n1 = ScalarParameter() {defaultValue=15, parameterName=EmissiveStrength}
  n2 = TextureSample() {texture=/VehicleCosmetics/Wheels/ChromeB/Textures/T_ChromeB_LensFlare}
  n3 = Multiply(A=@n2.A) {constB=0.2}
  MakeMaterialAttributes!(
   BaseColor
   Metallic
   Specular
   Roughness
   Anisotropy
   EmissiveColor=Multiply(A=Multiply(A=@n1, B=@n2.RGB), B=@n3)
   Opacity=@n3
   OpacityMask
   Normal
   Tangent
   WorldPositionOffset
   SubsurfaceColor
   ClearCoat
   ClearCoatRoughness
   AmbientOcclusion
   Refraction
   CustomizedUVs_0
   CustomizedUVs_1
   CustomizedUVs_2
   CustomizedUVs_3
   CustomizedUVs_4
   CustomizedUVs_5
   CustomizedUVs_6
   CustomizedUVs_7
   PixelDepthOffset
   ShadingModel
   Displacement
  )
 ]
}
